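# Niagara system for the BeanDog's glowing green eyes in WildEstateWeapons, drawn by one EyeGlows emitter as sprites plus dynamic lights.
# Burst-spawned particles persist while the emitter is active, tracking sampled skeletal mesh locations each update with a small offset.
# Pulsing waveforms drive sprite alpha and light color, each multiplied by a component of the sampled mesh scale, so glow intensity follows that scale.
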NiagaraSystem /WildEstateWeapons/Assets/FX/BeanDog/NS_WildEstate_BeanDog_Light
dependencies {
 usedDataInterfaces [&NiagaraDataInterfaceSkeletalMesh]
 usedDynamicInputs []
 usedModules [
  &self:SystemSpawnScript.SystemScriptSource.SystemScriptGraph.NiagaraNodeAssignment_1.SetVariables_#1
  &SystemState
 ]
}
emitters {EyeGlows {
  emitterUpdateScript [
   &EmitterState()
   &SpawnBurst_Instantaneous(Use Spawn Probability=false, UseLoopCountLimit=true)
  ]
  particleSpawnScript [
   &InitializeParticle(Color=<rgba 0.2488 2 0 0.15>, Sprite Size Mode=::NewEnumerator3, Uniform Sprite Size=42, UsePositionOffset=false, Write Material Random=true)
   &CameraOffset(Camera Offset Amount=-5)
  ]
  particleUpdateScript [
   &ParticleState(Kill Particles When Lifetime Has Elapsed=false, Let Infinitely Lived Particles Die When Emitter Deactivates=true)
   &SkeletalMeshLocation(Bone Sampling Mode=::NewEnumerator4, MaskSpawnGroup=false, Sampled Position Offset=<5 0 0>, Skeletal Mesh=$System.Skeletal Mesh Reader)
   &ScaleColor(
    Scale Alpha=&Multiply_Float(
     A=&Waveform("Waveform [1]"=::NewEnumerator0, "[1] Frequency"=0.5, "[1] Amplitude Min/Max"=<0.5 1>, "[1] Scale"=2)
     B=&MakeFloatFromVector(Vector=$Particles.SkeletalMeshLocation.SampledScale, Channel=::NewEnumerator0)
    )
    ScaleA=true
    ScaleRGB=false
   )
   &FN_Light_Attributes(
    Color=<rgba 1.0923 3 0 1>
    Create Color Attribute=true
    Create Exponent Attribute=true
    Create Radius Attribute=true
    Exponent=8
    Radius=50
    Scale Color=&Multiply_Float(
     A=&Waveform("Waveform [1]"=::NewEnumerator0, "[1] Frequency"=0.5, "[1] Amplitude Min/Max"=<0.25 1>, "[1] Scale"=4)
     B=&MakeFloatFromVector(Vector=$Particles.SkeletalMeshLocation.SampledScale, Channel=::NewEnumerator0)
    )
   )
  ]
  rendererClasses [&NiagaraSpriteRendererProperties, &NiagaraLightRendererProperties]
  renderers [
   {rendererClass=&NiagaraSpriteRendererProperties}
   {rendererClass=&NiagaraLightRendererProperties, rendererIndex=1}
  ]
 }}
summary {
 systemName=NS_WildEstate_BeanDog_Light
 emitters [{emitterName=EyeGlows, rendererClasses [&NiagaraSpriteRendererProperties, &NiagaraLightRendererProperties]}]
 userVariables []
}
